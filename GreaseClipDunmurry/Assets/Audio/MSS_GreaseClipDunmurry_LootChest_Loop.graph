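# Looping stereo MetaSound source for the GreaseClipDunmurry loot chest ambient sound.
# On play, a single stereo Wave Player starts the loot chest loop wave and feeds left and right straight to the outputs.
# Loop attenuation, loot-ambient sound class and concurrency settings govern playback.
# The Max Duck Amount input is declared but unconnected, so the graph itself applies no ducking.
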MetaSoundSource /GreaseClipDunmurry/Assets/Audio/MSS_GreaseClipDunmurry_LootChest_Loop
inputs [
 Max Duck Amount {type=Float}
 UE.Source.OnPlay {type=Trigger}
]
interfaces [UE.OutputFormat.Stereo:1.0, UE.Source:1.1]
outputs [
 "UE.OutputFormat.Stereo.Audio:0" {type=Audio}
 "UE.OutputFormat.Stereo.Audio:1" {type=Audio}
]
pages [Default {
  input_defaults {Max Duck Amount=0.216, UE.Source.OnPlay=null}
  nodes [
   InitVariable.WaveAsset:Array()
   WavePlayer_0 = UE.Wave Player:Stereo(Play=^UE.Source.OnPlay) {also {
     defaults {Loop="<MetaSoundBuilderResult.SUCCEEDED: 0>", Wave Asset=/GreaseClipDunmurry/Assets/Audio/Wavs/IP_GreaseClipDunmurry_LootChest_Loop}
     version [1,1]
    }}
  ]
  links [{dst="@UE.OutputFormat.Stereo.Audio:0", src=WavePlayer_0.Out Left}, {dst="@UE.OutputFormat.Stereo.Audio:1", src=WavePlayer_0.Out Right}]
 }]
properties {
 assetImportData=&self:AssetImportData
 attenuationSettings=&/Fort_GreaseClipDunmurry_Loop_Att
 bLooping=true
 duration=10000
 maxDistance=1250
 numChannels=2
 outputFormat=Stereo
 sampleRate=48000
 soundClassObject=/Game/Sounds/FortSoundClassesAndModes/Fort_Loot_Ambient_Loop
 volume=1.7
 concurrencySet [/Game/Sounds/Concurrency_Rules/Loot_Ambient_Conc]
}
