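# Boost sound cue for the GreaseClip Dunmurray vehicle, layering a start sound, a vehicle-specific loop and a quieter shared jet-wash loop.
# FuelRemaining drives continuous modulators that drop both loops' pitch and volume as fuel runs low.
# A nested modulator on the vehicle loop inversely raises pitch as fuel drains, before the outer low-fuel sag pulls it down.
# It uses shared vehicle boost attenuation and overrides concurrency with a per-owner limit.
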SoundCue /GreaseClipDunmurry/Assets/Audio/SC_Veh_GreaseClip_Dunmurray_Boost
root=Mixer_1
nodes [
 {id=Mixer_1, children [ModulatorContinuous_2, ModulatorContinuous_0, WavePlayer_0], properties {inputVolume [1,0.5,1]}}
 {
  id=ModulatorContinuous_0
  children [WavePlayer_4]
  properties {
   pitchModulationParams {maxInput=0.4, minOutput=0.75, parameterName=FuelRemaining}
   volumeModulationParams {maxInput=0.4, minOutput=0.75, parameterName=FuelRemaining}
  }
 }
 {
  id=ModulatorContinuous_1
  children [WavePlayer_1]
  properties {pitchModulationParams {maxOutput=0.8, minOutput=1.2, parameterName=FuelRemaining}}
 }
 {
  id=ModulatorContinuous_2
  children [ModulatorContinuous_1]
  properties {
   pitchModulationParams {maxInput=0.2, minOutput=0.7, parameterName=FuelRemaining}
   volumeModulationParams {maxInput=0.2, minOutput=0.5, parameterName=FuelRemaining}
  }
 }
 {id=WavePlayer_0, properties {soundWaveAssetPtr=/GreaseClipDunmurry/Assets/Audio/Wavs/IP_GreaseClip_Dunmurray_Boost_Start_01}}
 {id=WavePlayer_1, properties {bLooping=true, soundWaveAssetPtr=/GreaseClipDunmurry/Assets/Audio/Wavs/IP_GreaseClip_Dunmurray_Boost_Loop_01}}
 {id=WavePlayer_4, properties {bLooping=true, soundWaveAssetPtr=/VehicleAudio/Assets/Boost/Veh_Valet_Boost_Loop_Jet_Wash_01}}
]
properties {
 attenuationSettings=/VehicleAudio/Settings/Vehicle_Movement_Boost_Att
 bOverrideConcurrency=true
 maxDistance=12000
 soundClassObject=&Fort_Vehicles_Exterior
 volumeMultiplier=1
 concurrencyOverrides {bLimitToOwner=true, maxCount=2, volumeScale=0.5}
 concurrencySet [/VehicleAudio/Settings/Vehicle_Engine_Conc]
}
