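# Sound cue for the GreaseClip vehicle's pre-tunnel moment, playing the Dunmurry "TimeWarp Start" audio.
# A single wave player feeds the output directly, with no randomization or modulation, routed to the exterior vehicle sound class.
# Its own concurrency override caps instances per owner and scales their volume, taking precedence over the assigned vehicle engine concurrency set.
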SoundCue /GreaseClipDunmurry/Assets/Audio/SC_Veh_GreaseClip_PreTunnel
root=WavePlayer_0
nodes [{id=WavePlayer_0, properties {soundWaveAssetPtr=/GreaseClipDunmurry/Assets/Audio/Wavs/IP_GreaseClip_Dunmurray_TimeWarp_Start_01}}]
properties {
 bOverrideConcurrency=true
 maxDistance=2097152
 soundClassObject=&Fort_Vehicles_Exterior
 volumeMultiplier=1
 concurrencyOverrides {bLimitToOwner=true, maxCount=2, volumeScale=0.5}
 concurrencySet [/VehicleAudio/Settings/Vehicle_Engine_Conc]
}
