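# Niagara system for the GreaseClipDunmurry car's time-travel warm-up effect, with WarmUpAlpha ramping spawn rates and intensity.
# CPU-low and GPU wheel sparks orbit four sampled sockets with curl noise and rearward acceleration, alongside a glowing front cone with light and streaming sparks.
# Ribbon trails follow baked point-array curves, while mesh zaps spawn at preset body positions.
# User FXScale sizes effects, and IsTimeTravelSequence suppresses sparks and the cone burst.
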NiagaraSystem /GreaseClipDunmurry/Assets/FX/NS_GreaseClipDunMurry_WarmUp
dependencies {
 usedDataInterfaces [
  &NiagaraDataInterfaceArrayFloat3
  &NiagaraDataInterfaceColorCurve
  &NiagaraDataInterfaceCurve
  &NiagaraDataInterfaceArrayPosition
  &NiagaraDataInterfaceSocketReader
 ]
 usedDynamicInputs []
 usedModules [
  &self:SystemSpawnScript.SystemScriptSource.SystemScriptGraph.NiagaraNodeAssignment_3.SetVariables_#1
  &self:SystemSpawnScript.SystemScriptSource.SystemScriptGraph.NiagaraNodeAssignment_0.SetVariables_#2
  &self:SystemSpawnScript.SystemScriptSource.SystemScriptGraph.NiagaraNodeAssignment_1.SetVariables_#3
  &SystemState
  &self:SystemSpawnScript.SystemScriptSource.SystemScriptGraph.NiagaraNodeAssignment_2.SetVariables_#4
  &self:SystemSpawnScript.SystemScriptSource.SystemScriptGraph.NiagaraNodeAssignment_4.SetVariables_#5
 ]
}
emitters {
 NE_CPU_WheelSpark_LOW {
  emitterUpdateScript [
   &EmitterState()
   &SpawnRate(
    SpawnRate=&MakeCustomFloatFromBool(
     Bool=$User.IsTimeTravelSequence
     True Float=0
     False Float=&Lerp_Float(
      A=0
      B=50
      Alpha=&FloatFromCurve(FloatCurve=DataInterface() {propertyValues {Curve {keys [<at 0 0>,<at 1 1>]}}}, CurveIndex=$System.WarmUpAlpha)
     )
    )
    Use Spawn Probability=false
   )
  ]
  particleSpawnScript [
   &PartitionParticles(Number Of Partitions=4)
   &InitializeParticle(
    Color Mode=::NewEnumerator0
    Lifetime=&Lerp_Float(
     A=&RandomRangeFloat(Minimum=0.5, Maximum=0.75)
     B=&RandomRangeFloat(Minimum=0.25, Maximum=0.75)
     Alpha=$System.WarmUpAlpha
    )
    Position Offset=$System.FrontConeOffset
    Sprite Size=&Multiply_Vector2D_ByFloat(
     Float=&Multiply_Float(A=&RandomRangeFloat(Minimum=2, Maximum=3), B=$User.FXScale)
     Vector2D=<1 3>
    )
    Sprite Size Mode=::NewEnumerator4
    UsePositionOffset=true
    Write Material Random=true
   )
   /SetVariables(Particles.LRAlpha:NiagaraFloat=&ModuloFloat(A=&MakeFloatFromInt(Int=$Particles.PartitionParticles.Partition), B=2)) {isSetParameters=true, module=SetVariables_#6}
   &Color(Color=&ColorFromCurve(
     ColorCurve=DataInterface() {propertyValues {
       AlphaCurve {keys [<at 0 1 RCIM_Cubic>]}
       BlueCurve {keys [<at 0.4226 0.062 RCIM_Cubic>,<at 0.469 0.1501 RCIM_Cubic>]}
       GreenCurve {keys [<at 0.4226 4.1503 RCIM_Cubic>,<at 0.469 1.5881 RCIM_Cubic>]}
       RedCurve {keys [<at 0.4226 10 RCIM_Cubic>,<at 0.469 10 RCIM_Cubic>]}
      }}
     CurveIndex=&RandomRangeFloat(Minimum=0, Maximum=1)
    ))
   &AddVelocity(Velocity=<-1 0 0>, Velocity Speed Scale=&RandomRangeFloat(Minimum=550, Maximum=750)) {enabled=false}
  ]
  particleUpdateScript [
   &ParticleState()
   &SocketLocation(
    Position Sampling=::NewEnumerator1
    SamplePosOffset=false
    Socket Index=$Particles.PartitionParticles.Partition
    Socket Reader=$System.Socket Reader
    Socket Sampling Mode=::NewEnumerator3
    VelocitySpeedLimit=false
   )
   &RotateAroundPoint(
    Radius:NiagaraFloat=&Multiply_Float(A=&RandomRangeFloat(Minimum=40, Maximum=50), B=$User.FXScale)
    Rotation Center:NiagaraPosition=&AddVectorToPosition(
     Position=$Output.SocketLocation.SampledPosition
     Vector=&RotateVectorByQuaternion(
      Vector To Rotate=&Multiply_Vector(
       A=&Multiply_VectorByFloat(Vector=&RandomRangeVector(Minimum=<0 -15 0>, Maximum=<0 -20 0>), Float=$User.FXScale)
       B=&Lerp_Vector(A=<1 1 1>, B=<1 -1 1>, Alpha=$Particles.LRAlpha)
      )
      Quaternion=$Output.SocketLocation.SampledRotation
     )
    )
    Rotation Phase:NiagaraFloat=&Add_Float(
     A=&RandomRangeFloat(Minimum=0, Maximum=1)
     B=&Multiply_Float(A=$Particles.Age, B=&Lerp_Float(A=3, B=4, Alpha=$System.WarmUpAlpha))
    )
    Set Particle Position=false
    X Axis:Vector3f=&TransformVector(Vector=<1 0 0>, Destination Space=::Simulation)
    Y Axis:Vector3f=<0 0 1>
   )
   /SetVariables(Particles.Position=/Niagara/DynamicInputs/TypeConversions/MakeCustomPositionFromBool(
     Bool:NiagaraBool=&SetBoolByFloatComparison(A=$Particles.NormalizedAge, B=0.5, Comparison Type=::NewEnumerator6)
     True Position:NiagaraPosition=$Output.RotateAroundPoint.OutputPosition
     False Position:NiagaraPosition=$Particles.Position
    )) {isSetParameters=true, module=SetVariables_#7}
   &Drag(Use Linear Drag=true, Use Rotational Drag=true) {enabled=false}
   &AccelerationForce(Acceleration=&TransformVector(Vector=&Multiply_VectorByFloat(Vector=<-1500 0 0>, Float=$User.FXScale), Destination Space=::Simulation))
   &CurlNoiseForce(Bias Noise=false, Noise Strength=&Multiply_Float(A=1500, B=$User.FXScale), Pan Noise=false)
   &SolveForcesAndVelocity(Clamp Velocity=false, Limit Acceleration=false)
   &SpriteFacingAndAlignment(
    Sprite Alignment=&NormalizeVector(
     Vector To Normalize=&Subtract_Position(A=$Particles.Position, B=$Particles.Previous.Position)
     Safe Normalize=true
     Safe Fallback Threshold=0
     Safe Fallback Vector=<0 0 0.001>
    )
    WriteAlignment=true
    WriteFacing=false
   )
   &FN_Sprite_Attributes(
    Color=&MultiplyLinearColor(
     Linear Color=$Particles.Color
     Scale Factor=&MakeVector4_FromLinearColor(LinearColor=<rgba 0.5 0.5 0.5 0.01>)
    )
    Create Color Attribute=true
    Create SpriteSize Attribute=true
    SpriteSize=&Multiply_Vector2D(A=$Particles.SpriteSize, B=<15 5>)
    Write to Module Namespace=true
   )
  ]
  rendererClasses [&NiagaraSpriteRendererProperties]
  renderers [
   {rendererClass=&NiagaraSpriteRendererProperties}
   {rendererClass=&NiagaraSpriteRendererProperties, rendererIndex=1}
  ]
 }
 NE_ConeSparks {
  emitterUpdateScript [
   &EmitterState()
   &SpawnRate(
    SpawnRate=&MakeCustomFloatFromBool(
     Bool=$User.IsTimeTravelSequence
     True Float=0
     False Float=&Lerp_Float(
      A=0
      B=50
      Alpha=&FloatFromCurve(FloatCurve=DataInterface() {propertyValues {Curve {keys [<at 0.5 0>,<at 1 1>]}}}, CurveIndex=$System.WarmUpAlpha)
     )
    )
    Use Spawn Probability=false
   )
  ]
  particleSpawnScript [
   &InitializeParticle(
    Color Mode=::NewEnumerator0
    Lifetime=&Lerp_Float(
     A=&RandomRangeFloat(Minimum=0.5, Maximum=0.75)
     B=&RandomRangeFloat(Minimum=0.25, Maximum=0.75)
     Alpha=$System.WarmUpAlpha
    )
    Position Offset=$System.FrontConeOffset
    Sprite Size=&Multiply_Vector2D_ByFloat(
     Float=&Multiply_Float(
      A=&RandomRangeFloat(Minimum=2, Maximum=5)
      B=&Multiply_Float(A=$User.FXScale, B=1)
     )
     Vector2D=<1 4>
    )
    Sprite Size Mode=::NewEnumerator4
    UsePositionOffset=true
    Write Material Random=true
   )
   &Color(Color=&ColorFromCurve(
     ColorCurve=DataInterface() {propertyValues {
       AlphaCurve {keys [<at 0 1 RCIM_Cubic>]}
       BlueCurve {keys [<at 0.4226 10 RCIM_Cubic>,<at 0.469 10 RCIM_Cubic>]}
       GreenCurve {keys [<at 0.4226 3.2426 RCIM_Cubic>,<at 0.469 4.0456 RCIM_Cubic>]}
       RedCurve {keys [<at 0.4226 1.7244 RCIM_Cubic>,<at 0.469 2.1508 RCIM_Cubic>]}
      }}
     CurveIndex=&RandomRangeFloat(Minimum=0, Maximum=1)
    ))
  ]
  particleUpdateScript [
   &ParticleState()
   /SetVariables(Particles.Position=&AddVectorToPosition(
     Position=$Particles.Initial.Position
     Vector=&Multiply_Vector(
      A=&Add_Vector(
       A=&Lerp_Vector(
        A=<0 0 0>
        B=<-750 0 0>
        Alpha=&FloatFromCurve(FloatCurve=DataInterface() {propertyValues {Curve {keys [
             {arriveTangent=0.2266, interpMode=RCIM_Cubic, leaveTangent=0.2266, tangentMode=RCTM_User, time=0, value=0}
             {arriveTangent=0.8833, interpMode=RCIM_Cubic, leaveTangent=0.8833, tangentMode=RCTM_User, time=1, value=1}
            ]}}})
       )
       B=&Multiply_VectorByFloat(
        Vector=&RotateVectorByQuaternion(
         Vector To Rotate=<0 0 1>
         Quaternion=&MakeQuaternion(
          Quaternion From=0
          Axis=<1 0 0>
          Angle=&RandomRangeFloat(Minimum=0, Maximum=1)
          From=?
          To=?
          Angle Type=::NewEnumerator1
          Normalize Vectors=false
         )
        )
        Float=&Lerp_Float(
         A=0
         B=100
         Alpha=&FloatFromCurve(FloatCurve=DataInterface() {propertyValues {Curve {keys [
              {arriveTangent=3.7638, interpMode=RCIM_Cubic, leaveTangent=3.7638, tangentMode=RCTM_User, time=0, value=0}
              {interpMode=RCIM_Cubic, tangentMode=RCTM_User, time=1, value=1}
             ]}}})
        )
       )
      )
      B=$System.CarScale
     )
    )) {isSetParameters=true, module=SetVariables_#7}
   &SolveForcesAndVelocity(Clamp Velocity=false, Limit Acceleration=false)
   &SpriteFacingAndAlignment(
    Sprite Alignment=&NormalizeVector(
     Vector To Normalize=&Subtract_Position(A=$Particles.Position, B=$Particles.Previous.Position)
     Safe Normalize=true
     Safe Fallback Threshold=0
     Safe Fallback Vector=<0 0 0.001>
    )
    WriteAlignment=true
    WriteFacing=false
   )
   &FN_Sprite_Attributes(
    Color=&MultiplyLinearColor(
     Linear Color=$Particles.Color
     Scale Factor=&MakeVector4_FromLinearColor(LinearColor=<rgba 0.5 0.5 0.5 0.01>)
    )
    Create Color Attribute=true
    Create SpriteSize Attribute=true
    SpriteSize=&Multiply_Vector2D(A=$Particles.SpriteSize, B=<10 3>)
    Write to Module Namespace=true
   )
  ]
  rendererClasses [&NiagaraSpriteRendererProperties]
  renderers [
   {rendererClass=&NiagaraSpriteRendererProperties}
   {rendererClass=&NiagaraSpriteRendererProperties, rendererIndex=1}
  ]
 }
 NE_FrontCone001 {
  emitterUpdateScript [
   &EmitterState()
   &SpawnRate(SpawnRate=&Lerp_Float(A=0, B=5, Alpha=$System.WarmUpAlpha), Use Spawn Probability=false)
   &SpawnBurst_Instantaneous(
    Spawn Count=&MakeCustomIntFromBool(Bool=$User.IsTimeTravelSequence, True INT=0, False INT=1)
    Spawn Group=1
    Spawn Time=0.01
    Use Spawn Probability=false
    UseLoopCountLimit=false
   )
  ]
  particleSpawnScript [
   &InitializeParticle(
    Lifetime=&MaskFloatBySpawnGroup(Masked Float=1, Passthrough Float=1)
    Mesh Scale=&Multiply_Vector(A=$System.CarScale, B=<0.5 1 0.75>)
    Mesh Scale Mode=::NewEnumerator4
    Position Offset=&AddVectorToPosition(
     Position=Linked() {linkedVariable {name=System.FrontConeOffset, type=Vector3f}}
     Vector=&RandomRangeVector(Minimum=<0 0 0>, Maximum=<0 0 0>)
    )
    UsePositionOffset=true
    Write Material Random=true
   )
   /SetVariables(Particles.VisibilityTag=&SpawnGroup()) {isSetParameters=true}
   &InitialMeshOrientation(
    Mesh Orientation Mode=::NewEnumerator7
    Rotation=&RandomRangeVector(Minimum=<0 0 0>, Maximum=<1 0 0>)
    Use Rotation Vector=true
   )
  ]
  particleUpdateScript [
   &ParticleState()
   &ScaleMeshSize(Scale Factor=&Lerp_Vector(
     A=<2 2 2>
     B=<6 1.5 1.5>
     Alpha=&FloatFromCurve(FloatCurve=DataInterface() {propertyValues {Curve {keys [
          {interpMode=RCIM_Cubic, tangentMode=RCTM_User, time=0, value=0}
          {arriveTangent=1.9569, interpMode=RCIM_Cubic, leaveTangent=1.9569, tangentMode=RCTM_User, time=1, value=1}
         ]}}})
    ))
   &DynamicMaterialParameters(
    Index 0 Param 2=1
    Index 1 Param 3=&Lerp_Float(
     A=1.5
     B=0.7
     Alpha=&FloatFromCurve(
      FloatCurve=DataInterface() {propertyValues {Curve {keys [
          {interpMode=RCIM_Cubic, tangentMode=RCTM_User, time=0, value=0}
          {arriveTangent=1.5894, interpMode=RCIM_Cubic, leaveTangent=1.5894, tangentMode=RCTM_User, time=1, value=1}
         ]}}}
      CurveIndex=$System.WarmUpAlpha
     )
    )
    Module.Index0Float0Write=true
    Module.Index0Float1Write=true
    Module.Index0Float2Write=true
    Module.Index0Float3Write=true
    Module.Index1Float0Write=true
    Module.Index1Float1Write=true
    Module.Index1Float2Write=true
    Module.Index1Float3Write=true
    Write Parameter Index 1=::NewEnumerator1
   )
   &Color(Color=&Multiply_LinearColorByFloat(
     Linear Color=&Multiply_LinearColorByFloat(
      Linear Color=&ColorFromCurve(
       ColorCurve=DataInterface() {propertyValues {
         AlphaCurve {keys [<at 0 1 RCIM_Cubic>]}
         BlueCurve {keys [<at 0.4077 1 RCIM_Cubic>,<at 1 2 RCIM_Cubic>]}
         GreenCurve {keys [<at 0.4077 1 RCIM_Cubic>,<at 1 2 RCIM_Cubic>]}
         RedCurve {keys [<at 0.4077 1 RCIM_Cubic>,<at 1 2 RCIM_Cubic>]}
        }}
       CurveIndex=$System.WarmUpAlpha
      )
      Float=&FloatFromCurve(
       FloatCurve=DataInterface() {propertyValues {Curve {keys [{interpMode=RCIM_Cubic, tangentMode=RCTM_User, time=0, value=0}, {interpMode=RCIM_Cubic, tangentMode=RCTM_User, time=0.2, value=1}]}}}
       CurveIndex=$Engine.System.Age
      )
     )
     Float=&FloatFromCurve(
      FloatCurve=DataInterface() {propertyValues {Curve {keys [
          {arriveTangent=4.9846, interpMode=RCIM_Cubic, leaveTangent=4.9846, tangentMode=RCTM_User, time=0, value=0}
          {interpMode=RCIM_Cubic, tangentMode=RCTM_User, time=0.25, value=1}
          {interpMode=RCIM_Cubic, tangentMode=RCTM_User, time=1, value=0}
         ]}}}
      Scale Curve=&FN_Select_Float_from_Time_Of_Day(Morning=0.8, Day=1, Evening=0.8, Night=0.7)
     )
    ))
   /SetVariables(Particles.ColorHighlight:LinearColor=&Multiply_LinearColorByFloat(Linear Color=$Particles.Color, Float=0.75)) {isSetParameters=true}
   &FN_Sprite_Attributes(
    Camera Offset=50
    Color=&Multiply_LinearColorByFloat(
     Linear Color=<rgba 0.06 0.0684 0.25 1>
     Float=&FloatFromCurve(
      FloatCurve=DataInterface() {propertyValues {Curve {keys [
          {arriveTangent=4.7347, interpMode=RCIM_Cubic, leaveTangent=4.7347, tangentMode=RCTM_User, time=0, value=0}
          {interpMode=RCIM_Cubic, tangentMode=RCTM_User, time=0.25, value=1}
          {interpMode=RCIM_Cubic, tangentMode=RCTM_User, time=1, value=0}
         ]}}}
      Scale Curve=&FN_Select_Float_from_Time_Of_Day(Morning=1, Day=1, Evening=1, Night=0.75)
     )
    )
    Create Camera Offset Attribute=true
    Create Color Attribute=true
    Create Position Attribute=true
    Create SpriteSize Attribute=true
    Position:NiagaraPosition=&AddVectorToPosition(Position=$Particles.Position, Vector=<-15 0 5>)
    SpriteSize=&Multiply_Vector2D_ByFloat(Float=1000, Vector2D=<1 1>)
    Write to Module Namespace=true
   )
   &FN_Sprite_Attributes(
    Camera Offset=50
    Color=&Multiply_LinearColorByFloat(
     Linear Color=<rgba 0.2399 0.2734 1 1>
     Float=&FloatFromCurve(
      FloatCurve=DataInterface() {propertyValues {Curve {keys [
          {arriveTangent=4.7347, interpMode=RCIM_Cubic, leaveTangent=4.7347, tangentMode=RCTM_User, time=0, value=0}
          {interpMode=RCIM_Cubic, tangentMode=RCTM_User, time=0.25, value=1}
          {interpMode=RCIM_Cubic, tangentMode=RCTM_User, time=1, value=0}
         ]}}}
      Scale Curve=&FN_Select_Float_from_Time_Of_Day(Morning=0.75, Day=0.75, Evening=0.75, Night=0.5)
     )
    )
    Create Camera Offset Attribute=true
    Create Color Attribute=true
    Create Position Attribute=true
    Create SpriteSize Attribute=true
    Position:NiagaraPosition=&AddVectorToPosition(Position=$Particles.Position, Vector=<0 0 5>)
    SpriteSize=&Multiply_Vector2D_ByFloat(Float=750, Vector2D=<1 1>)
    Write to Module Namespace=true
   ) {module=FN_Sprite_Attributes001}
   &FN_Light_Attributes(
    Color=<rgba 0.1919 0.3399 1 1>
    Create Color Attribute=true
    Create Exponent Attribute=true
    Create Position Attribute=true
    Create Radius Attribute=true
    Enable Particle Light Enable=true
    Exponent=1
    Light Enabled=true
    Position=&AddVectorToPosition(Position=$Particles.Position, Vector=<-40 0 0>)
    Radius=250
    Scale Color=&FloatFromCurve(FloatCurve=DataInterface() {propertyValues {Curve {keys [<at 0 0>,<at 0.5 1>]}}}, CurveIndex=$Engine.System.Age, Scale Curve=25)
   )
  ]
  rendererClasses [
   &NiagaraMeshRendererProperties
   &NiagaraSpriteRendererProperties
   &NiagaraLightRendererProperties
  ]
  renderers [
   {rendererClass=&NiagaraMeshRendererProperties, rendererIndex=0}
   {rendererClass=&NiagaraMeshRendererProperties, rendererIndex=1}
   {rendererClass=&NiagaraSpriteRendererProperties, rendererIndex=2}
   {rendererClass=&NiagaraSpriteRendererProperties, rendererIndex=3}
   {rendererClass=&NiagaraLightRendererProperties, rendererIndex=4}
  ]
 }
 NE_GPU_WheelSparks {
  simTarget=GPUComputeSim
  emitterUpdateScript [
   &EmitterState()
   &SpawnRate(
    SpawnRate=&MakeCustomFloatFromBool(
     Bool=$User.IsTimeTravelSequence
     True Float=0
     False Float=&Lerp_Float(
      A=0
      B=250
      Alpha=&FloatFromCurve(FloatCurve=DataInterface() {propertyValues {Curve {keys [<at 0 0>,<at 1 1>]}}}, CurveIndex=$System.WarmUpAlpha)
     )
    )
    Use Spawn Probability=false
   )
  ]
  particleSpawnScript [
   &PartitionParticles(Number Of Partitions=4)
   &InitializeParticle(
    Color Mode=::NewEnumerator0
    Lifetime=&Lerp_Float(
     A=&RandomRangeFloat(Minimum=0.5, Maximum=0.75)
     B=&RandomRangeFloat(Minimum=0.25, Maximum=0.75)
     Alpha=$System.WarmUpAlpha
    )
    Position Offset=$System.FrontConeOffset
    Sprite Size=&Multiply_Vector2D_ByFloat(
     Float=&Multiply_Float(A=&RandomRangeFloat(Minimum=2, Maximum=3), B=$User.FXScale)
     Vector2D=<1 3>
    )
    Sprite Size Mode=::NewEnumerator4
    UsePositionOffset=true
    Write Material Random=true
   )
   /SetVariables(Particles.LRAlpha:NiagaraFloat=&ModuloFloat(A=&MakeFloatFromInt(Int=$Particles.PartitionParticles.Partition), B=2)) {isSetParameters=true, module=SetVariables_#6}
   &Color(Color=&ColorFromCurve(
     ColorCurve=DataInterface() {propertyValues {
       AlphaCurve {keys [<at 0 1 RCIM_Cubic>]}
       BlueCurve {keys [<at 0.4226 0.062 RCIM_Cubic>,<at 0.469 0.1501 RCIM_Cubic>]}
       GreenCurve {keys [<at 0.4226 4.1503 RCIM_Cubic>,<at 0.469 1.5881 RCIM_Cubic>]}
       RedCurve {keys [<at 0.4226 10 RCIM_Cubic>,<at 0.469 10 RCIM_Cubic>]}
      }}
     CurveIndex=&RandomRangeFloat(Minimum=0, Maximum=1)
    ))
   &AddVelocity(Velocity=<-1 0 0>, Velocity Speed Scale=&RandomRangeFloat(Minimum=550, Maximum=750)) {enabled=false}
  ]
  particleUpdateScript [
   &ParticleState()
   &SocketLocation(
    Position Sampling=::NewEnumerator1
    SamplePosOffset=false
    Socket Index=$Particles.PartitionParticles.Partition
    Socket Reader=$System.Socket Reader
    Socket Sampling Mode=::NewEnumerator3
    VelocitySpeedLimit=false
   )
   &RotateAroundPoint(
    Radius:NiagaraFloat=&Multiply_Float(A=&RandomRangeFloat(Minimum=40, Maximum=50), B=$User.FXScale)
    Rotation Center:NiagaraPosition=&AddVectorToPosition(
     Position=$Output.SocketLocation.SampledPosition
     Vector=&RotateVectorByQuaternion(
      Vector To Rotate=&Multiply_Vector(
       A=&Multiply_VectorByFloat(Vector=&RandomRangeVector(Minimum=<0 -15 0>, Maximum=<0 -20 0>), Float=$User.FXScale)
       B=&Lerp_Vector(A=<1 1 1>, B=<1 -1 1>, Alpha=$Particles.LRAlpha)
      )
      Quaternion=$Output.SocketLocation.SampledRotation
     )
    )
    Rotation Phase:NiagaraFloat=&Add_Float(
     A=&RandomRangeFloat(Minimum=0, Maximum=1)
     B=&Multiply_Float(A=$Particles.Age, B=&Lerp_Float(A=3, B=4, Alpha=$System.WarmUpAlpha))
    )
    Set Particle Position=false
    X Axis:Vector3f=&TransformVector(Vector=<1 0 0>, Destination Space=::Simulation)
    Y Axis:Vector3f=<0 0 1>
   )
   /SetVariables(Particles.Position=&MakeCustomPositionFromBool(
     Bool:NiagaraBool=&SetBoolByFloatComparison(A=$Particles.NormalizedAge, B=0.5, Comparison Type=::NewEnumerator6)
     True Position:NiagaraPosition=$Output.RotateAroundPoint.OutputPosition
     False Position:NiagaraPosition=$Particles.Position
    )) {isSetParameters=true, module=SetVariables_#7}
   &Drag(Use Linear Drag=true, Use Rotational Drag=true) {enabled=false}
   &AccelerationForce(Acceleration=&TransformVector(Vector=&Multiply_VectorByFloat(Vector=<-1500 0 0>, Float=$User.FXScale), Destination Space=::Simulation))
   &CurlNoiseForce(Bias Noise=false, Noise Strength=&Multiply_Float(A=1500, B=$User.FXScale), Pan Noise=false)
   &SolveForcesAndVelocity(Clamp Velocity=false, Limit Acceleration=false)
   &SpriteFacingAndAlignment(
    Sprite Alignment=&NormalizeVector(
     Vector To Normalize=&Subtract_Position(A=$Particles.Position, B=$Particles.Previous.Position)
     Safe Normalize=true
     Safe Fallback Threshold=0
     Safe Fallback Vector=<0 0 0.001>
    )
    WriteAlignment=true
    WriteFacing=false
   )
   &FN_Sprite_Attributes(
    Color=&MultiplyLinearColor(
     Linear Color=$Particles.Color
     Scale Factor=&MakeVector4_FromLinearColor(LinearColor=<rgba 0.5 0.5 0.5 0.01>)
    )
    Create Color Attribute=true
    Create SpriteSize Attribute=true
    SpriteSize=&Multiply_Vector2D(A=$Particles.SpriteSize, B=<15 5>)
    Write to Module Namespace=true
   )
  ]
  rendererClasses [&NiagaraSpriteRendererProperties]
  renderers [
   {rendererClass=&NiagaraSpriteRendererProperties}
   {rendererClass=&NiagaraSpriteRendererProperties, rendererIndex=1}
  ]
 }
 NE_SparkTrails {
  emitterSpawnScript [
   /SetVariables(Emitter.CurveAmount=10) {isSetParameters=true}
   /SetVariables(Emitter.ParticlePerRibbon=10) {isSetParameters=true}
  ]
  emitterUpdateScript [
   &EmitterState(
    Delay First Loop Only=true
    Life Cycle Mode=::NewEnumerator1
    Loop Duration=&Lerp_Float(A=0.5, B=0.1, Alpha=$System.WarmUpAlpha)
    Recalculate Duration Each Loop=true
    UseLoopDelay=true
   )
   &SpawnBurst_Instantaneous(
    Spawn Count=&MakeCustomIntFromBool(Bool=&SetBoolByFloatComparison(A=$System.WarmUpAlpha, B=0), True INT=$Emitter.ParticlePerRibbon, False INT=0)
    Use Spawn Probability=false
    UseLoopCountLimit=false
   )
   &SpawnRate(SpawnGroup=1, SpawnRate=10, Use Spawn Probability=false) {enabled=false}
  ]
  particleSpawnScript [
   &PartitionParticles(
    Normalized Index In Partition Includes One=true
    Particles Per Partition:NiagaraInt32=$Emitter.ParticlePerRibbon
    Partition Count Mode:ENiagara_InfiniteLoopDuration=::NewEnumerator1
    Partition Mode:ENiagara_PartitionMode_v2=::NewEnumerator1
    Write Partition to Ribbon ID=::NewEnumerator0
   )
   /SetVariables(Particles.RandomIntPerLoop:NiagaraInt32=&ModuloInt(A=$Emitter.LoopCount, B=7)) {isSetParameters=true}
   /SetVariables(Particles.Age:NiagaraFloat=&Add_Float(A=&Lerp_Float(A=0, B=-0.2, Alpha=$Particles.PartitionParticles.Normalized Index In Partition), B=0)) {isSetParameters=true}
   /SetVariables(Particles.VisibilityTag=&SpawnGroup()) {isSetParameters=true}
   &InitializeParticle(
    Color Mode=::NewEnumerator0
    Lifetime=/Game/Effects/Niagara/DynamicInputs/Random/FN_SeededRandomRangeFloat(Minimum:NiagaraFloat=0.3, Maximum:NiagaraFloat=0.5, Random Seed:NiagaraInt32=$Particles.RandomIntPerLoop)
    Ribbon Width=&Multiply_Float(A=5, B=$User.FXScale)
    Ribbon Width Mode=::NewEnumerator1
    Sprite Size Mode=::NewEnumerator3
    Uniform Sprite Size=&MakeCustomFloatFromBool(
     Bool=&SetBoolByIntComparison(A=$Particles.PartitionParticles.Index In Partition, B=1, Comparison Type=::NewEnumerator6)
     True Float=&Multiply_Float(A=150, B=$User.FXScale)
     False Float=0
    )
    UsePositionOffset=false
    Write Material Random=true
   )
   /SetVariables(Particles.RibbonLinkOrder:NiagaraFloat=$Particles.PartitionParticles.Normalized Index In Partition) {isSetParameters=true}
  ]
  particleUpdateScript [
   &ParticleState()
   &self:SampleCurveArray(
    Alpha:NiagaraFloat=$Particles.NormalizedAge
    CurveSampler:NiagaraInt32=$Particles.RandomIntPerLoop
    PointPerCurve=12
    VectorArray=DataInterface() {propertyValues {
      GpuSyncMode=SyncCpuToGpu
      MaxElements=0
      FloatData [
       <311.4351 -35.3157 31.5638>
       <276.0575 -114.095 54.3639>
       <210.3671 -146.6109 71.6855>
       <145.5506 -107.8378 94.3424>
       <131.659 -32.4056 109.5923>
       <112.1646 32.423 115.6584>
       <53.9275 20.321 134.1906>
       <1.0275 -34.7965 147.8509>
       <-85.7576 -15.4043 156.2088>
       <-152.733 34.1016 139.3023>
       <-231.6657 0.3717 114.614>
       <-316.6544 1.2993 136.2086>
       <299.967 58.2971 47.9921>
       <252.9536 111.4351 54.2745>
       <190.6002 134.2151 60.3283>
       <119.8957 140.4683 66.2577>
       <94.194 95.5767 111.8261>
       <26.9965 94.754 125.9538>
       <-24.9893 118.8018 106.5968>
       <-68.322 151.9755 68.9474>
       <-139.0522 147.1013 62.1021>
       <-189.5172 116.8458 49.4186>
       <-243.8728 91.6541 45.6881>
       <-314.7736 87.287 50.9205>
       <311.111 -13.9747 57.1433>
       <259.8929 27.2812 73.849>
       <219.796 14.6063 90.1894>
       <184.0164 -34.7416 106.1885>
       <118.139 -30.9221 122>
       <59.5028 -9.6507 137.421>
       <10.5156 34.8764 152.3215>
       <-49.4595 44.6936 160.0365>
       <-110.7838 21.2481 152.443>
       <-167.5242 -4.6036 139.0266>
       <-223.4293 31.5556 152.1246>
       <-287.9388 36.1564 141.8395>
       <303.9001 -23.6854 46.6263>
       <240.738 -58.9321 69.716>
       <193.7595 -43.0848 88.708>
       <163.2825 24.8609 103.5219>
       <104.0087 53.5853 123.3821>
       <46.4766 16.894 146.0132>
       <9.4905 -48.8478 139.5412>
       <-61.3707 -54.2698 147.8711>
       <-106.7851 -4.6226 154.632>
       <-150.579 48.0488 155.0194>
       <-224.6961 53.8587 139.5987>
       <-299.1508 54.511 147.2522>
       <302.541 -45.1282 44.7244>
       <269.8542 -99.1596 25.4638>
       <217.0777 -126.7397 37.5734>
       <161.4761 -150.5486 34.3375>
       <100.2412 -150.0022 23.456>
       <62.88 -127.1403 58.7652>
       <2.3384 -131.1857 66.2157>
       <-63.1384 -139.5724 67.3283>
       <-114.4241 -148.2349 44.2154>
       <-168.7921 -132.241 67.337>
       <-223.7154 -116.2487 89.5088>
       <-289.305 -108.7137 89.4459>
       <299.8531 24.357 33.2681>
       <272.6206 83.3225 26.3291>
       <230.3077 122.1438 25.7814>
       <171.4425 141.1037 43.6539>
       <109.4576 136.8645 47.9377>
       <49.4898 136.9468 63.1439>
       <-10.6092 140.8449 71.5744>
       <-67.3819 155.6575 82.3901>
       <-127.7242 164.5859 94.4281>
       <-185.547 151.4832 73.1517>
       <-243.1123 147.0061 56.35>
       <-305.3303 165.294 64.1705>
       <297.345 -21.6997 56.2757>
       <277.0865 -86.9796 59.2929>
       <227.1379 -120.4982 60.9606>
       <161.7263 -129.6858 66.741>
       <101.2099 -111.449 92.9358>
       <35.2657 -114.4033 97.1255>
       <-30.5674 -132.0398 94.5917>
       <-97.8192 -137.9846 96.1871>
       <-164.1 -123.0402 91.7221>
       <-230.2174 -122.0281 97.7095>
       <-275.5887 -80.3078 89.8739>
       <-335.2011 -73.5734 91.7169>
      ]
     }}
   ) {module=SampleCurveArray}
   /SetVariables(Particles.Position=&Multiply_VectorByFloat(Vector=$Output.SampleCurveArray.CurveSampledVector, Float=$User.FXScale)) {isSetParameters=true}
   &Color(Color=&MultiplyLinearColor(
     Linear Color=&ColorFromCurve(ColorCurve=DataInterface() {propertyValues {
        AlphaCurve {keys [<at 0 0 RCIM_Cubic>,<at 0.1049 1 RCIM_Cubic>,<at 0.581 1>,<at 1 0>]}
        BlueCurve {keys [<at 0.0827 1 RCIM_Cubic>,<at 0.9036 1 RCIM_Cubic>]}
        GreenCurve {keys [<at 0.0827 0.3982 RCIM_Cubic>,<at 0.9036 0.1359 RCIM_Cubic>]}
        RedCurve {keys [<at 0.0827 0.1796 RCIM_Cubic>,<at 0.9036 0 RCIM_Cubic>]}
       }})
     Scale Factor=&MakeVector4_FromLinearColor(LinearColor=&ColorFromCurve(
       ColorCurve=DataInterface() {propertyValues {
         AlphaCurve {keys [<at 0 1 RCIM_Cubic>,<at 1 0 RCIM_Cubic>]}
         BlueCurve {keys [<at 0 1 RCIM_Cubic>,<at 0 50 RCIM_Cubic>,<at 0.1708 50>]}
         GreenCurve {keys [<at 0 1 RCIM_Cubic>,<at 0 50 RCIM_Cubic>,<at 0.1708 50>]}
         RedCurve {keys [<at 0 1 RCIM_Cubic>,<at 0 50 RCIM_Cubic>,<at 0.1708 50>]}
        }}
       CurveIndex=$Particles.PartitionParticles.Normalized Index In Partition
      ))
    )) {module=Color001}
   &FN_Sprite_Attributes(
    Color=&MultiplyLinearColor(
     Linear Color=&ColorFromCurve(ColorCurve=DataInterface() {propertyValues {
        AlphaCurve {keys [<at 0 0 RCIM_Cubic>,<at 0.1049 1 RCIM_Cubic>,<at 0.581 1>,<at 1 0>]}
        BlueCurve {keys [<at 0.0827 10 RCIM_Cubic>,<at 0.8905 10 RCIM_Cubic>]}
        GreenCurve {keys [<at 0.0827 2.8795 RCIM_Cubic>,<at 0.8905 3.3987 RCIM_Cubic>]}
        RedCurve {keys [<at 0.0827 2.7799 RCIM_Cubic>,<at 0.8905 1.9192 RCIM_Cubic>]}
       }})
     Scale Factor=<1 1 1 1>
    )
    Create Color Attribute=true
    Write to Module Namespace=true
   )
   /SetVariables(
    Particles.GlowRibbonColor:LinearColor=&MultiplyLinearColor(
     Linear Color=$Particles.Color
     Scale Factor=&MakeVector4FromVectorAndFloat("Vector (RGB)"=<1 1 1>, "Float (Alpha)"=0.06)
    )
    Particles.GlowRibbonWidth:NiagaraFloat=&Multiply_Float(A=$Particles.RibbonWidth, B=4)
   ) {isSetParameters=true}
   /SetVariables(
    Particles.DecalCamOffset=50
    Particles.DecalColor:LinearColor=&ColorFromCurve(ColorCurve=DataInterface() {propertyValues {
       AlphaCurve {keys [<at 0.0012 0>,<at 0.0291 1 RCIM_Cubic>,<at 0.9653 1>,<at 1 0 RCIM_Cubic>]}
       BlueCurve {keys [<at 0 15 RCIM_Cubic>]}
       GreenCurve {keys [<at 0 6.1051 RCIM_Cubic>]}
       RedCurve {keys [<at 0 3 RCIM_Cubic>]}
      }})
   ) {isSetParameters=true}
   /SetVariables(Particles.GlowDecalSize:Vector2f=&Multiply_Vector2D_ByFloat(Float=1.5, Vector2D=$Particles.SpriteSize)) {isSetParameters=true}
   /SetVariables(Particles.DecalDynMat:Vector4f=<1 1 0 0>) {isSetParameters=true}
   &FN_Light_Attributes(
    Color=&ColorFromCurve(ColorCurve=DataInterface() {propertyValues {
       AlphaCurve {keys [<at 0.0012 0>,<at 0.0291 1 RCIM_Cubic>,<at 0.8885 1>,<at 1 0 RCIM_Cubic>]}
       BlueCurve {keys [<at 0.0994 15>, {interpMode=RCIM_Cubic, leaveTangent=-17.7855, time=0.8541, value=15}]}
       GreenCurve {keys [<at 0.0994 4.9933>, {interpMode=RCIM_Cubic, leaveTangent=-5.9205, time=0.8541, value=4.9933}]}
       RedCurve {keys [<at 0.0994 1.5>, {interpMode=RCIM_Cubic, leaveTangent=-1.7785, time=0.8541, value=1.5}]}
      }})
    Create Color Attribute=true
    Create Exponent Attribute=true
    Create Radius Attribute=true
    Enable Particle Light Enable=true
    Exponent=1.5
    Light Enabled:NiagaraBool=&SetBoolByIntComparison(A=$Particles.PartitionParticles.Index In Partition, B=1, Comparison Type=::NewEnumerator6)
    Radius=&FloatFromCurve(FloatCurve=DataInterface() {propertyValues {Curve {keys [<at 0 0>,<at 0.1 1>,<at 0.9 1>,<at 1 0>]}}}, Scale Curve=100)
    Scale Color=&FloatFromCurve(FloatCurve=DataInterface() {propertyValues {Curve {keys [<at 0 0>,<at 0.2 1>,<at 0.65 1>,<at 1 0>]}}}, Scale Curve=5)
   )
  ]
  rendererClasses [
   &NiagaraRibbonRendererProperties
   &NiagaraSpriteRendererProperties
   &NiagaraLightRendererProperties
  ]
  renderers [
   {rendererClass=&NiagaraRibbonRendererProperties}
   {rendererClass=&NiagaraSpriteRendererProperties, rendererIndex=1}
   {rendererClass=&NiagaraSpriteRendererProperties, rendererIndex=2}
   {rendererClass=&NiagaraRibbonRendererProperties, rendererIndex=3}
   {rendererClass=&NiagaraSpriteRendererProperties, rendererIndex=4}
   {rendererClass=&NiagaraLightRendererProperties, rendererIndex=5}
  ]
 }
 NE_Zaps {
  emitterUpdateScript [
   &EmitterState()
   &SpawnRate(
    SpawnRate=&Lerp_Float(
     A=0
     B=50
     Alpha=&FloatFromCurve(FloatCurve=DataInterface() {propertyValues {Curve {keys [<at 0 0>,<at 1 1>]}}}, CurveIndex=$System.WarmUpAlpha)
    )
    Use Spawn Probability=false
   )
  ]
  particleSpawnScript [
   &InitializeParticle(
    Lifetime=&MaskFloatBySpawnGroup(Masked Float=&RandomRangeFloat(Minimum=0.4, Maximum=0.5), Passthrough Float=0.5)
    Mesh Scale Mode=::NewEnumerator3
    Mesh Uniform Scale=&Multiply_Float(A=&RandomRangeFloat(Minimum=0.2, Maximum=0.4), B=$User.FXScale)
    Sprite Size Mode=::NewEnumerator3
    Uniform Sprite Size=20
    UsePositionOffset=false
    Write Material Random=true
   )
   /SetVariables(Particles.Position=&Multiply_VectorByFloat(
     Vector=/Niagara/DynamicInputs/Arrays/SelectPositionFromArray(
      Array Sampling Mode:ENiagaraArraySamplingMode=::NewEnumerator0
      Direct Array Index:NiagaraInt32=?
      Interpolation:NiagaraFloat=?
      Direct Array Mode:ENiagara_WrapClamp=::NewEnumerator0
      Evaluation Type:ENiagaraRandomnessEvaluation=::NewEnumerator0
      Normalized Interpolation:NiagaraBool=true
      Recalculate Random Each Loop:NiagaraBool=false
      Position Array:NiagaraDataInterfaceArrayPosition=DataInterface() {propertyValues {
        GpuSyncMode=SyncCpuToGpu
        MaxElements=0
        PositionData [
         <192.3417 -106.3385 56.2992>
         <92.6447 -122.9211 43.791>
         <-13.0827 109.5889 17.7729>
         <142.0064 0.1204 93.4829>
         <29.4106 -113.9767 25.7462>
         <116.2699 111.1617 26.0482>
         <117.2309 -109.3109 33.1014>
         <-177.3799 83.9769 40.9377>
         <249.6831 -21.8321 21.3158>
         <3.9345 81.9685 97.2264>
         <168.9551 54.324 86.5875>
         <222.9013 -19.8242 72.0191>
         <246.6383 -48.9313 22.7249>
         <-95.7748 -93.6136 132.8081>
         <-193.8617 15.2701 116.4681>
         <241.4438 29.5988 32.5795>
         <131.6617 -103.186 86.5738>
         <166.6924 -42.0642 87.723>
         <221.2293 32.9973 72.3963>
         <104.2041 -104.7772 15.3916>
         <37.5161 -119.9627 65.6404>
         <-163.4495 -100.2206 114.2653>
         <206.3187 -31.4263 77.7426>
         <-35.2132 -56.8897 138.6964>
         <-102.0604 -116.6031 86.6792>
         <9.4449 -94.3526 18.9676>
         <-61.4735 122.3095 81.8644>
         <172.3155 33.537 89.8953>
         <-2.5427 -125.9877 51.4818>
         <118.8014 -66.6175 95.3649>
         <255.069 -13.6363 57.9673>
         <-160.624 -81.3733 122.5638>
         <210.9325 -4.3932 82.2372>
         <153.522 -103.3489 83.1329>
         <-222.3238 21.796 107.5272>
         <184.1791 -95.5657 79.2045>
         <-215.3556 76.6847 73.8825>
         <237.3978 11.1751 66.4705>
         <-181.7926 -61.7395 113.1287>
         <150.9861 22.0087 93.7708>
         <-140.5879 -108.2859 110.999>
         <-187.8447 -109.8502 82.3367>
         <-31.5395 -116.0174 27.5626>
         <-82.8378 -118.3023 101.884>
         <119.7693 84.2747 93.5966>
         <-160.3679 -116.5473 76.8586>
         <246.1191 69.8206 36.1936>
         <166.8189 8.5702 88.7516>
         <60.768 122.0367 57.4476>
         <-155.4475 52.9177 123.1371>
         <131.294 16.9034 96.3179>
         <261.5865 19.8383 38.1771>
         <-62.2791 -120.3975 87.0202>
         <-52.4525 -95.134 19.9279>
         <-9.6441 -120.3448 69.4248>
         <160.6771 104.4142 80.701>
         <-213.9655 -37.8039 72.0559>
         <57.494 -119.2154 66.2876>
         <269.1017 54.5713 43.8613>
         <59.2696 -101.5779 93.1854>
         <-42.2465 117.2781 84.1296>
         <190.8019 -7.8766 82.8127>
         <-201.9474 55.1033 110.944>
         <-12.6958 -108.1583 92.674>
         <170.2475 -109.5751 70.5293>
         <112.6074 65.8097 96.5755>
         <-216.1889 56.5053 73.0691>
         <204.5027 38.7701 83.298>
         <-139.8525 -84.3482 126.6283>
         <71.0756 -112.9317 79.6364>
         <141.8727 -110.2615 68.1724>
         <180.8964 -27.7131 88.2717>
         <-126.5531 -99.3776 123.4007>
         <5.9633 103.9868 14.1969>
         <-63.6801 -128.6146 65.5422>
         <68.7898 -113.0872 25.1632>
         <-46.3381 -108.4863 100.5344>
         <245.4742 93.4478 34.7195>
         <16.3784 -124.8169 44.9802>
         <-81.0464 -125.9165 80.1332>
         <254.1365 -36.0837 37.2053>
         <-158.2162 -17.1685 123.7894>
         <-8.4272 37.9598 138.425>
         <119.9362 -113.8702 71.5067>
         <62.3185 88.6571 96.4136>
         <-145.2928 102.6469 117.2994>
         <-104.5249 11.8364 137.4016>
         <-152.76 110.8747 96.6106>
         <46.951 116.9957 30.8236>
         <-117.1088 91.4811 129.4365>
         <123.4621 118.8279 55.5483>
         <-190.2108 -37.2764 118.0378>
         <76.1045 103.8814 91.5534>
         <55.6949 107.4453 88.0284>
         <-153.2521 85.0017 123.6615>
         <-222.9449 -72.0291 92.8439>
         <-181.7458 52.9996 116.9386>
         <-187.341 -3.9136 47.4172>
         <-149.4652 9.8206 20.9387>
         <-128.4521 56.3787 129.2046>
        ]
       }}
      Randomness Mode:ENiagaraRandomnessMode=::NewEnumerator0
      Fixed Random Seed:NiagaraBool=false
      Random Seed:NiagaraInt32=0
     )
     Float=$User.FXScale
    )) {isSetParameters=true}
   &InitialMeshOrientation(
    Mesh Orientation Mode=::NewEnumerator2
    Orientation Vector=&RandomVector(Vector Scale=1)
    Use Rotation Vector=true
   )
   &CameraOffset(Camera Offset Amount=10)
  ]
  particleUpdateScript [
   &ParticleState()
   &ScaleMeshSize(Scale Factor=&Multiply_VectorByFloat(
     Vector=&RandomRangeVector(Minimum=<1 1 2.2>, Maximum=<1.3 1.3 2.2>)
     Float=&FloatFromCurve(FloatCurve=DataInterface() {propertyValues {Curve {keys [<at 0 1>,<at 1 1.5>]}}})
    )) {module=ScaleMeshSize001}
   &DynamicMaterialParameters(
    Index 0 Param 2=&RandomRangeFloat(Minimum=0, Maximum=0)
    Index 0 Param 4=&MaskFloatBySpawnGroup(Masked Float=13, Passthrough Float=1)
    Index 1 Param 1=&FloatFromCurve(FloatCurve=DataInterface() {propertyValues {Curve {keys [<at 0 0.5>,<at 1 1>]}}})
    Index 1 Param 2=2
    Index 1 Param 3=&FloatFromCurve(FloatCurve=DataInterface() {propertyValues {Curve {keys [{interpMode=RCIM_Cubic, tangentMode=RCTM_Break, time=0.5, value=-0.5}, <at 1 0.1>]}}})
    Index 2 Param 1=&Multiply_Float(
     A=&FloatFromCurve(
      FloatCurve=DataInterface() {propertyValues {Curve {keys [{interpMode=RCIM_Cubic, leaveTangent=8.1085, tangentMode=RCTM_Break, time=0.8568, value=1}, {tangentMode=RCTM_User, time=1, value=0}]}}}
      Scale Curve=&MakeCustomFloatFromBool(Bool=$User.IsTimeTravelSequence, True Float=0.7, False Float=1)
     )
     B=&Divide_Float(
      A=&MakeFloatFromVector(Vector=$Particles.Scale, Channel=::NewEnumerator0)
      B=&RandomRangeFloat(Minimum=10, Maximum=15)
     )
    )
    Index 2 Param 2=&FloatFromCurve(
     FloatCurve=DataInterface() {propertyValues {Curve {keys [
         {interpMode=RCIM_Cubic, tangentMode=RCTM_Break, time=0.7, value=0}
         {arriveTangent=-0.0859, interpMode=RCIM_Cubic, tangentMode=RCTM_Break, time=1, value=-1}
        ]}}}
     Scale Curve=50
    )
    Index 2 Param 3=&Waveform(
     Global Input Position=&RandomRangeFloat(Minimum=0, Maximum=15)
     Global Phase=$Particles.Age
     "Waveform [1]"=::NewEnumerator12
     "[1] Phase Scale"=2
     "[1] Frequency"=1
     "[1] Amplitude Min/Max"=<0 1>
    )
    Index 2 Param 4=25
    Index 3 Param 1=&FloatFromCurve(
     FloatCurve=DataInterface() {propertyValues {Curve {keys [
         {interpMode=RCIM_Cubic, leaveTangent=-0.0025, tangentMode=RCTM_Break, time=0, value=0}
         {arriveTangent=1.8528, interpMode=RCIM_Cubic, tangentMode=RCTM_Break, time=1, value=1}
        ]}}}
     Scale Curve=&RandomRangeFloat(Minimum=0, Maximum=0)
    )
    Index 3 Param 2=&FloatFromCurve(
     FloatCurve=DataInterface() {propertyValues {Curve {keys [
         {interpMode=RCIM_Cubic, leaveTangent=-0.0025, tangentMode=RCTM_Break, time=0, value=0.5}
         {arriveTangent=-0.0316, interpMode=RCIM_Cubic, tangentMode=RCTM_Break, time=1, value=-0.2}
        ]}}}
     Scale Curve=0.1
    )
    Index 3 Param 3=&FloatFromCurve(FloatCurve=DataInterface() {propertyValues {Curve {keys [
         {interpMode=RCIM_Cubic, leaveTangent=1.1034, tangentMode=RCTM_Break, time=0, value=-1}
         {arriveTangent=2.2399, tangentMode=RCTM_User, time=0.7, value=0.125}
        ]}}})
    Index 3 Param 4=&FloatFromCurve(FloatCurve=DataInterface() {propertyValues {Curve {keys [
         {interpMode=RCIM_Cubic, leaveTangent=3.0601, tangentMode=RCTM_Break, time=0.5, value=0}
         {arriveTangent=3.2053, tangentMode=RCTM_User, time=1, value=1}
        ]}}})
    Module.Index0Param0Write=true
    Module.Index0Param1Write=true
    Module.Index0Param2Write=true
    Module.Index0Param3Write=true
    Module.Index1Param0Write=true
    Module.Index1Param1Write=true
    Module.Index1Param2Write=true
    Module.Index1Param3Write=true
    Module.Index2Param0Write=true
    Module.Index2Param1Write=true
    Module.Index2Param2Write=true
    Module.Index2Param3Write=true
    Module.Index3Param0Write=true
    Module.Index3Param1Write=true
    Module.Index3Param2Write=true
    Module.Index3Param3Write=true
    Write Parameter Index 1=true
    Write Parameter Index 2=true
    Write Parameter Index 3=true
   )
   &SolveForcesAndVelocity(Clamp Velocity=false, Limit Acceleration=false)
   &Color(
    Color=&ColorFromCurve(ColorCurve=DataInterface() {propertyValues {
       AlphaCurve {keys [<at 0 1 RCIM_Cubic>,<at 0.7884 1>,<at 1 0 RCIM_Cubic>]}
       BlueCurve {keys [<at 0.2225 5>, {interpMode=RCIM_Cubic, leaveTangent=-19.7604, time=0.4906, value=3}]}
       GreenCurve {keys [<at 0.2225 2.6274>, {interpMode=RCIM_Cubic, leaveTangent=-19.7604, time=0.4906, value=0.8233}]}
       RedCurve {keys [<at 0.2225 1.7465>, {interpMode=RCIM_Cubic, leaveTangent=-19.7604, time=0.4906, value=0.6782}]}
      }})
    Scale Alpha=&FloatFromCurve(FloatCurve=DataInterface() {propertyValues {Curve {keys [
         {interpMode=RCIM_Cubic, leaveTangent=2.5294, tangentMode=RCTM_Break, time=0, value=0}
         {arriveTangent=-0.2166, interpMode=RCIM_Cubic, tangentMode=RCTM_Break, time=0.4, value=1}
        ]}}})
   )
   /SetVariables(Particles.DecalColor:LinearColor=&MultiplyLinearColor(Linear Color=<rgba 1 1 1 1>, Scale Factor=&MakeVector4_FromLinearColor(LinearColor=$Particles.Color))) {isSetParameters=true}
   /SetVariables(Particles.GlowDecalSize:Vector2f=&Multiply_Vector2D_ByFloat(Float=&Multiply_Float(A=100, B=$User.FXScale), Vector2D=<1 1>)) {isSetParameters=true}
   /SetVariables(Particles.DecalDynMat:Vector4f=<7 1 0 0>) {isSetParameters=true}
  ]
  rendererClasses [&NiagaraMeshRendererProperties, &NiagaraSpriteRendererProperties]
  renderers [
   {rendererClass=&NiagaraMeshRendererProperties, rendererIndex=0}
   {rendererClass=&NiagaraSpriteRendererProperties, rendererIndex=1}
   {rendererClass=&NiagaraMeshRendererProperties, rendererIndex=2}
  ]
 }
}
summary {
 systemName=NS_GreaseClipDunMurry_WarmUp
 emitters [
  {
   emitterName=NE_SparkTrails
   rendererClasses [
    &NiagaraRibbonRendererProperties
    &NiagaraSpriteRendererProperties
    &NiagaraLightRendererProperties
   ]
  }
  {emitterName=NE_ConeSparks, rendererClasses [&NiagaraSpriteRendererProperties]}
  {emitterName=NE_Zaps, rendererClasses [&NiagaraMeshRendererProperties, &NiagaraSpriteRendererProperties]}
  {emitterName=NE_GPU_WheelSparks, simTarget=GPUComputeSim, rendererClasses [&NiagaraSpriteRendererProperties]}
  {
   emitterName=NE_FrontCone001
   rendererClasses [
    &NiagaraMeshRendererProperties
    &NiagaraSpriteRendererProperties
    &NiagaraLightRendererProperties
   ]
  }
  {emitterName=NE_CPU_WheelSpark_LOW, rendererClasses [&NiagaraSpriteRendererProperties]}
 ]
 userVariables [
  User.FXScale {type=NiagaraFloat, defaultValue=1, description=""}
  User.IntroDuration {type=NiagaraFloat, defaultValue=0, description=""}
  User.IsTimeTravelSequence {type=NiagaraBool}
 ]
}
