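# Niagara system for a camera teleport effect, showing a travel tunnel with streaking speed lines.
# TunnelCone places a mesh at System.ConeOffset, rotates it by TunnelRotation, feeds FadeOut to its material and kills it by normalized timeline time.
# SpeedlinesChroma bursts sprites that cycle a wrapped ProgressAlpha, lerping from along the tunnel axis to randomized points near the cone, aligned to their motion.
# Its TravelAlpha-driven cone scale lerps 1 to 1, so has no effect.
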NiagaraSystem /GreaseClipDunmurry/Assets/FX/NS_GreaseClipDunmurry_CameraTeleport
dependencies {
 usedDataInterfaces [&NiagaraDataInterfaceCurve]
 usedDynamicInputs [
  &Divide_Float
  &Boolean_NotOperation
  /Game/Effects/Niagara/DynamicInputs/State/FN_IsExecutionStateActive
 ]
 usedModules [
  &self:SystemSpawnScript.SystemScriptSource.SystemScriptGraph.NiagaraNodeAssignment_1.SetVariables_#1
  &self:SystemSpawnScript.SystemScriptSource.SystemScriptGraph.NiagaraNodeAssignment_5.SetVariables_#2
  &self:SystemSpawnScript.SystemScriptSource.SystemScriptGraph.NiagaraNodeAssignment_6.SetVariables_#3
  &SystemState
  &self:SystemSpawnScript.SystemScriptSource.SystemScriptGraph.NiagaraNodeAssignment_3.SetVariables_#4
  &self:SystemSpawnScript.SystemScriptSource.SystemScriptGraph.NiagaraNodeAssignment_0.SetVariables_#5
  &Timeline
  &self:SystemSpawnScript.SystemScriptSource.SystemScriptGraph.NiagaraNodeAssignment_2.SetVariables_#6
 ]
}
emitters {
 SpeedlinesChroma {
  emitterUpdateScript [
   &EmitterState(
    Inactive Response=::NewEnumerator1
    Life Cycle Mode=::NewEnumerator1
    Loop Behavior=::NewEnumerator1
    Loop Duration Mode=::NewEnumerator1
    UseLoopDelay=false
   )
   &SpawnBurst_Instantaneous(Spawn Count=30, Use Spawn Probability=false, UseLoopCountLimit=false)
  ]
  particleSpawnScript [
   &InitializeParticle(
    Color=<rgba 10 10 10 1>
    Lifetime Max=1
    Lifetime Min=0.5
    Lifetime Mode=::NewEnumerator1
    Sprite Size=&Multiply_Vector2D(A=<3 800>, B=&RandomRangeVector2D(Minimum=<0.5 0.5>, Maximum=<1 1>))
    Sprite Size Mode=::NewEnumerator4
    UsePositionOffset=false
    Write Material Random=true
   )
   /SetVariables(Particles.ProgressAlpha:NiagaraFloat=&RandomRangeFloat(Minimum=0, Maximum=1)) {isSetParameters=true}
  ]
  particleUpdateScript [
   &ParticleState(Kill Particles When Lifetime Has Elapsed=false)
   /SetVariables(Particles.ProgressAlpha:NiagaraFloat=&ModuloFloat(
     A=&Add_Float(A=$Particles.ProgressAlpha, B=&RandomRangeFloat(Minimum=0.01, Maximum=0.03))
     B=1
    )) {isSetParameters=true}
   /SetVariables(Particles.Position=&Lerp_Vector(
     A=&AddVectorToPosition(
      Position=Linked() {linkedVariable {name=System.ConeOffset, type=Vector3f}}
      Vector=&RotateVectorByQuaternion(Vector To Rotate=<-10000 0 0>, Quaternion=$System.TunnelRotation)
     )
     B=&AddVectorToPosition(
      Position=&AddVectorToPosition(Position=Linked() {linkedVariable {name=System.ConeOffset, type=Vector3f}}, Vector=<0 0 300>)
      Vector=&RotateVectorByQuaternion(
       Vector To Rotate=<0 600 0>
       Quaternion=&MakeQuaternion(
        Quaternion From=0
        Axis=<1 0 0>
        Angle=&Add_Float(
         A=&RandomRangeFloat(Minimum=0, Maximum=1)
         B=&Multiply_Float(
          A=1
          B=&Multiply_Float(
           A=0.1
           B=&Lerp_Float(
            A=1
            B=-1
            Alpha=/Niagara/DynamicInputs/Integer/V2/Round_Float(Float:NiagaraFloat=&RandomRangeFloat(Minimum=0, Maximum=1))
           )
          )
         )
        )
        From=?
        To=?
        Angle Type=::NewEnumerator1
        Normalize Vectors=false
       )
      )
     )
     Alpha=&FloatFromCurve(FloatCurve=DataInterface() {propertyValues {Curve {keys [<at 0 0>,<at 1 1>]}}}, CurveIndex=$Particles.ProgressAlpha)
    )) {isSetParameters=true}
   &SpriteFacingAndAlignment(
    Sprite Alignment=&NormalizeVector(
     Vector To Normalize=&Subtract_Position(A=$Particles.Position, B=$Particles.Previous.Position)
     Safe Normalize=true
     Safe Fallback Threshold=0
     Safe Fallback Vector=<0 0 0.001>
    )
    WriteAlignment=true
    WriteFacing=false
   )
   &SolveForcesAndVelocity(Clamp Velocity=false, Limit Acceleration=false)
   &DynamicMaterialParameters(
    Index 0 Param 2=&Add_Float(
     A=&RandomRangeFloat(Minimum=0, Maximum=1)
     B=&FloatFromCurve(FloatCurve=DataInterface() {propertyValues {Curve {keys [<at 0 0>,<at 1 1>]}}})
    )
    Index 0 Param 3=0.5
    Module.Index0Float0Write=true
    Module.Index0Float1Write=true
    Module.Index0Float2Write=true
    Module.Index0Float3Write=true
   )
   &FN_Sprite_Attributes(
    Create Color Attribute=true
    Create SpriteSize Attribute=true
    SpriteSize=&Multiply_Vector2D(
     A=$Particles.SpriteSize
     B=&MaskVector2dBySpawnGroup(Masked Vector2D=<15 1.2>, Passthrough Vector2D=<8 1.5>, Spawn Group Mask=0)
    )
    Write to Module Namespace=true
   )
  ]
  rendererClasses [&NiagaraSpriteRendererProperties]
  renderers [
   {rendererClass=&NiagaraSpriteRendererProperties}
   {rendererClass=&NiagaraSpriteRendererProperties, rendererIndex=1}
  ]
 }
 TunnelCone {
  emitterUpdateScript [
   &EmitterState(Life Cycle Mode=::NewEnumerator1, Loop Behavior=::NewEnumerator1, Loop Duration Mode=::NewEnumerator1, UseLoopDelay=false)
   &SpawnBurst_Instantaneous(Use Spawn Probability=false, UseLoopCountLimit=false)
  ]
  particleSpawnScript [
   &InitializeParticle(
    Color Mode=::NewEnumerator0
    Lifetime=5
    Mesh Scale=<100 100 100>
    Mesh Scale Mode=::NewEnumerator4
    Position Offset=$System.ConeOffset
    UsePositionOffset=true
    Write Material Random=true
   )
   &InitialMeshOrientation(Mesh Orientation Mode=::NewEnumerator7, Rotation=<0 0 -0.5>, Use Rotation Vector=true)
  ]
  particleUpdateScript [
   &ParticleState(Kill Particles When Lifetime Has Elapsed=false)
   /SetVariables(Particles.MeshOrientation=&MultiplyQuaternion(Quaternion A=$Particles.InitialRotation, Quaternion B=$System.TunnelRotation)) {isSetParameters=true}
   &ScaleMeshSize(Scale Factor=&MakeVector(X=1, Y=&Lerp_Float(A=1, B=1, Alpha=$System.TravelAlpha), Z=1))
   /SetVariables(Particles.Position:Vector3f=$System.ConeOffset) {isSetParameters=true}
   &DynamicMaterialParameters(
    Index 0 Param 1=&Lerp_Float(A=0, B=1, Alpha=&Multiply_Float(A=$System.FadeOut, B=1))
    Module.Index0Float0Write=true
    Module.Index0Float1Write=true
    Module.Index0Float2Write=true
    Module.Index0Float3Write=true
   )
   &KillParticles(Kill Particles=&SetBoolByFloatComparison(A=$System.Timeline.NormalizedTime, B=0.98))
  ]
  rendererClasses [&NiagaraMeshRendererProperties]
  renderers [{rendererClass=&NiagaraMeshRendererProperties, rendererIndex=0}]
 }
}
summary {
 systemName=NS_GreaseClipDunmurry_CameraTeleport
 emitters [
  {emitterName=SpeedlinesChroma, rendererClasses [&NiagaraSpriteRendererProperties]}
  {emitterName=TunnelCone, rendererClasses [&NiagaraMeshRendererProperties]}
 ]
 userVariables [
  User.IsLocal {type=NiagaraBool}
  User.TravelDuration {type=NiagaraFloat, defaultValue=5.5, description=""}
 ]
}
