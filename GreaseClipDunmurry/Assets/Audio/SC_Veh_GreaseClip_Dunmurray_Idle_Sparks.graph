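# Sound cue for the idle spark sounds of the GreaseClip Dunmurray vehicle, played quietly as exterior close-detail audio.
# A Random node picks among seven spark waves and five empty, equally weighted slots, so some triggers stay silent.
# A root Modulator varies volume widely and pitch upward.
# It uses vehicle close-detail attenuation, its own concurrency override that blocks new voices beyond a small limit, and a pre-effect vehicle boost bus send.
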SoundCue /GreaseClipDunmurry/Assets/Audio/SC_Veh_GreaseClip_Dunmurray_Idle_Sparks
root=Modulator_0
nodes [
 {id=Modulator_0, children [Random_0], properties {pitchMax=1.5, pitchMin=1, volumeMax=1, volumeMin=0.1}}
 {
  id=Random_0
  children [WavePlayer_1,WavePlayer_2,WavePlayer_3,WavePlayer_4,WavePlayer_5,WavePlayer_6,WavePlayer_7,null,null,null,null,null]
  properties {weights [1,1,1,1,1,1,1,1,1,1,1,1]}
 }
 {id=WavePlayer_1, properties {soundWaveAssetPtr=/GreaseClipDunmurry/Assets/Audio/Wavs/GreaseClip_Dunmurray_Idle_Sparks_01}}
 {id=WavePlayer_2, properties {soundWaveAssetPtr=/GreaseClipDunmurry/Assets/Audio/Wavs/GreaseClip_Dunmurray_Idle_Sparks_02}}
 {id=WavePlayer_3, properties {soundWaveAssetPtr=/GreaseClipDunmurry/Assets/Audio/Wavs/GreaseClip_Dunmurray_Idle_Sparks_03}}
 {id=WavePlayer_4, properties {soundWaveAssetPtr=/GreaseClipDunmurry/Assets/Audio/Wavs/GreaseClip_Dunmurray_Idle_Sparks_04}}
 {id=WavePlayer_5, properties {soundWaveAssetPtr=/GreaseClipDunmurry/Assets/Audio/Wavs/GreaseClip_Dunmurray_Idle_Sparks_05}}
 {id=WavePlayer_6, properties {soundWaveAssetPtr=/GreaseClipDunmurry/Assets/Audio/Wavs/GreaseClip_Dunmurray_Idle_Sparks_06}}
 {id=WavePlayer_7, properties {soundWaveAssetPtr=/GreaseClipDunmurry/Assets/Audio/Wavs/GreaseClip_Dunmurray_Idle_Sparks_07}}
]
properties {
 attenuationSettings=/VehicleAudio/Settings/Vehicle_Movement_Close_Detail_Att
 bOverrideConcurrency=true
 maxDistance=12000
 soundClassObject=&Fort_Vehicles_Exterior
 volumeMultiplier=0.2
 concurrencyOverrides {maxCount=3, resolutionRule=PreventNew, volumeScale=0.5}
 concurrencySet [/VehicleAudio/Settings/Vehicle_Engine_Conc]
 preEffectBusSends [{audioBus=/VehicleAudio/Cues/AB_Vehicle_Boost, soundSourceBus=null}]
}
