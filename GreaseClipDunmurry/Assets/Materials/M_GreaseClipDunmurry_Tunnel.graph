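# Unlit material for a scrolling, glowing blue energy tunnel on mesh particles.
# A panned normal map drives fake directional lighting, combined with Fresnel-offset caustics and multiplied by panning distortion.
# Two spark layers, a V-based background gradient and noise-broken core glow add to the emissive.
# Separate branches hold a dynamic-parameter fade to FadeColor, an emissive-adjustment function and a VertexHide lerp toward negated world position that collapses vertices.
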Material /GreaseClipDunmurry/Assets/Materials/M_GreaseClipDunmurry_Tunnel
graph {
 properties {
  bAutomaticallySetUsageInEditor=false
  bUsedWithMeshParticles=true
  bUsedWithNiagaraMeshParticles=true
  shadingModel=MSM_Unlit
  usedShadingModels=Unlit
 }
 nodes [
  LinearInterpolate(A=Constant(), B=Multiply(A=WorldPosition().XYZ) {constB=-1}, Alpha=Ceil(None=ScalarParameter() {parameterName=VertexHide}))
  n1 = TextureSample(UVs=Panner(Coordinate=TextureCoordinate() {vTiling=0.5}) {speedX=-0.05, speedY=-0.24}) {
   samplerSource=SSM_Wrap_WorldGroupSettings
   samplerType=SAMPLERTYPE_Normal
   texture=/GreaseClipDunmurry/Assets/Materials/Textures/T_MemoryTunnel_N1
  }
  n2 = DotProduct(A=@n1.RGB, B=Normalize(VectorInput=VectorParameter() {parameterName=LightDirection, defaultValue=<rgba 0 -1 0.4 0>}.RGB))
  n3 = Saturate(None=@n2)
  n4 = Multiply(
   A=LinearInterpolate(
    A=VectorParameter() {parameterName=DimColor, defaultValue=<rgba 0.0437 0.4179 1 1>}.RGB
    B=VectorParameter() {parameterName=LightColor, defaultValue=<rgba 0.1683 0.2307 1 1>}.RGB
    Alpha=Power(Base=@n3) {constExponent=2.5}
   )
   B=Clamp(None=@n2) {minDefault=0.1}
  )
  n5 = MaterialFunctionCall(Float2=TextureCoordinate()) {materialFunction=&BreakOutFloat2Components}
  n6 = Multiply(A=Fresnel(Normal=NamedRerouteUsage()) {baseReflectFraction=0, exponent=2}) {constB=0.25}
  n7 = TextureSample!(
   UVs=Panner(Coordinate=AppendVector(A=@n5.R, B=@n6)) {speedX=0.04, speedY=0.1}
   Tex
   Bias=ScalarParameter() {parameterName=EmipBias}
   Apply View MipBias
  ) {mipValueMode=TMVM_MipBias, samplerSource=SSM_Wrap_WorldGroupSettings, texture=/Game/Effects/Fort_Effects/Textures/Tile/T_Caustics_Chromatic}
  n8 = Multiply(A=@n7.RGB) {constB=0.05}
  n9 = VectorParameter() {parameterName=DistortionTile, defaultValue=<rgba 1 1.5 0 0>}
  n10 = NamedRerouteUsage()
  n11 = Add(
   A=Multiply(A=TextureCoordinate() {uTiling=2, vTiling=2.5}, B=AppendVector(A=@n9.R, B=@n9.R))
   B=Multiply(A=ComponentMask(None=@n10) {g=true, r=true}) {constB=0.05}
  )
  n12 = TextureSampleParameter2D(UVs=Panner(Coordinate=@n11) {speedX=0.05, speedY=-0.6}) {
   parameterName=DistortionTexture
   samplerSource=SSM_Wrap_WorldGroupSettings
   texture=/Game/Effects/Fort_Effects/Textures/Tile/T_ShieldCaustics_Tile_Chromatic
  }
  n13 = OneMinus(None=MaterialFunctionCall(Float2=TextureCoordinate()) {materialFunction=&BreakOutFloat2Components}.G)
  n14 = Add(
   A=Multiply(A=Add(A=@n4, B=@n8), B=NamedRerouteDeclaration(None=@n12.RGB) {name=Distortion, nodeColor=<rgba 0 1 0.2734 1>})
   B=Multiply(A=Power(Base=@n13) {constExponent=1.5}, B=VectorParameter() {parameterName=BGColor, defaultValue=<rgba 0.0004 0.0212 0.1 1>}.RGB)
  )
  n15 = Constant3Vector() {constant=<rgba 1 0.415 0.0061 1>}
  n16 = Constant3Vector() {constant=<rgba 1 0.7185 0.125 1>}
  n17 = NamedRerouteUsage()
  n18 = Add(A=Multiply(A=ComponentMask(None=@n17) {g=true, r=true}) {constB=0.04}, B=TextureCoordinate())
  n19 = Constant2Vector() {g=1.5, r=3}
  n20 = TextureSample!(UVs=Panner(Coordinate=Multiply(A=@n18, B=@n19)) {speedX=0.05, speedY=-0.8}, Tex, Apply View MipBias) {
   samplerSource=SSM_Wrap_WorldGroupSettings
   samplerType=SAMPLERTYPE_Alpha
   texture=/GreaseClipDunmurry/Assets/Materials/Textures/T_GreaseClip_Dunmurry_Sparks
  }
  n21 = NamedRerouteUsage()
  n22 = Add(A=Multiply(A=ComponentMask(None=@n21) {g=true, r=true}) {constB=0.03}, B=TextureCoordinate())
  n23 = Constant2Vector() {g=2, r=4}
  n24 = TextureSample(UVs=Panner(Coordinate=Multiply(A=@n22, B=@n23)) {speedX=-0.02, speedY=-1.2}) {
   samplerSource=SSM_Wrap_WorldGroupSettings
   samplerType=SAMPLERTYPE_Alpha
   texture=&T_GreaseClip_Dunmurry_Sparks
  }
  n25 = Max(A=@n20.R, B=@n24.R)
  n26 = Multiply(A=Multiply(A=LinearInterpolate(A=@n15, B=@n16, Alpha=@n25), B=@n25), B=ScalarParameter() {defaultValue=0.5, parameterName=SparksIntensity})
  n27 = AppendVector(
   A=MaterialFunctionCall(Float2=TextureCoordinate()) {materialFunction=&BreakOutFloat2Components}.R
   B=Constant()
  )
  n28 = LinearInterpolate(Alpha=TextureSample(UVs=Panner(Coordinate=@n27) {speedX=0.04, speedY=0.3}) {samplerSource=SSM_Wrap_WorldGroupSettings, texture=/Game/Global/Textures/Default/Noise/T_Tiling_Noise_01}.R) {constA=0.1}
  n29 = MaterialFunctionCall(Float2=TextureCoordinate()) {materialFunction=&BreakOutFloat2Components}
  n30 = Multiply(A=Power(Base=OneMinus(None=@n29.G)) {constExponent=15}) {constB=20}
  n31 = Multiply(A=VectorParameter() {parameterName=CoreLightColorA, defaultValue=<rgba 0.2423 0.3231 1 1>}.RGB, B=Saturate(None=Multiply(A=@n28, B=@n30)))
  n32 = ScalarParameter() {defaultValue=3, parameterName=CoreEmissiveBrightness}
  n33 = Multiply(A=Add(A=Add(A=@n14, B=@n26), B=Multiply(A=@n31, B=@n32) {constB=2}), B=ScalarParameter() {defaultValue=3, parameterName=EmissiveBrightness})
  n34 = DynamicParameter() {paramNames [FadeIn,FadeOut,Param3,Param4]}
  n35 = Min(A=Divide(A=@n34.FadeIn) {constB=0.05}, B=Divide(A=@n34.FadeIn) {constB=0.2})
  LinearInterpolate(A=@n33, B=VectorParameter() {parameterName=FadeColor, defaultValue=<rgba 0.85 0.8788 1 1>}.RGB, Alpha=OneMinus(None=Saturate(None=@n35)))
  MaterialFunctionCall(Emissive=@n33) {materialFunction=&MF_FX_EmissiveAdjustments}
  n36 = TextureSampleParameter2D(UVs=Panner(Coordinate=@n11) {speedX=-0.025, speedY=-0.5}) {
   parameterName=DistortionTexture
   samplerSource=SSM_Wrap_WorldGroupSettings
   texture=&T_ShieldCaustics_Tile_Chromatic
  }
  Max(A=@n12.RGB, B=@n36.RGB)
  NamedRerouteDeclaration(None=Transform(None=NamedRerouteDeclaration(None=@n1.RGB) {name=TangentSpaceNormal, nodeColor=<rgba 1 0 0.7969 1>})) {name=WorldSpaceNormal, nodeColor=<rgba 0 1 0.9297 1>}
 ]
}
