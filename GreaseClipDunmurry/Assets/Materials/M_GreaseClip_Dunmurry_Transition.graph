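# Translucent unlit glow material for the GreaseClip Dunmurry transition effect, flagged for Niagara mesh particles.
# Panned noise, sampled from U alone to form streaks, is masked by the larger of two steep inverted-V ramps offset by FadeIn and FadeOut.
# The mask tints CoreLightColorA, scaled by CoreEmissiveBrightness into MF_FX_EmissiveAdjustments.
# A nonzero VertexHide switches a lerp to negated world position, which collapses vertices if used as position offset.
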Material /GreaseClipDunmurry/Assets/Materials/M_GreaseClip_Dunmurry_Transition
graph {
 properties {
  bAutomaticallySetUsageInEditor=false
  bUsedWithNiagaraMeshParticles=true
  blendMode=BLEND_Translucent
  previewMesh=/Game/Effects/Fort_Effects/Mesh/Festivus/SM_Festivus_ScreenCone_01
  shadingModel=MSM_Unlit
  usedShadingModels=Unlit
 }
 nodes [
  LinearInterpolate(A=Constant(), B=Multiply(A=WorldPosition().XYZ) {constB=-1}, Alpha=Ceil(None=ScalarParameter() {parameterName=VertexHide}))
  n1 = VectorParameter() {parameterName=CoreLightColorA, defaultValue=<rgba 0.4273 0.5398 0.9652 1>}
  n2 = MaterialFunctionCall(Float2=Multiply(A=TextureCoordinate()) {constB=2}) {materialFunction=&BreakOutFloat2Components}
  n3 = Constant()
  n4 = TextureSample(UVs=Panner(Coordinate=AppendVector(A=@n2.R, B=@n3)) {speedX=0.25, speedY=0.65}) {texture=/Game/Global/Textures/Default/Noise/T_Tiling_Noise_01}
  n5 = OneMinus(None=MaterialFunctionCall(Float2=TextureCoordinate()) {materialFunction=&BreakOutFloat2Components}.G)
  n6 = LinearInterpolate(Alpha=ScalarParameter() {defaultValue=1, parameterName=FadeIn}) {constA=-1}
  n7 = Multiply(A=Power(Base=Add(A=@n5, B=@n6)) {constExponent=20}) {constB=50}
  n8 = OneMinus(None=MaterialFunctionCall(Float2=TextureCoordinate()) {materialFunction=&BreakOutFloat2Components}.G)
  n9 = LinearInterpolate(Alpha=ScalarParameter() {defaultValue=0.5, parameterName=FadeOut}) {constA=-1}
  n10 = Multiply(A=Power(Base=Add(A=@n8, B=@n9)) {constExponent=20}) {constB=50}
  n11 = Saturate(None=Multiply(A=LinearInterpolate(Alpha=@n4.RGB) {constA=0.025}, B=Max(A=@n7, B=@n10)))
  MaterialFunctionCall(Emissive=Multiply(A=Multiply(A=@n1.RGB, B=@n11), B=ScalarParameter() {defaultValue=3, parameterName=CoreEmissiveBrightness}) {constB=2}) {materialFunction=&MF_FX_EmissiveAdjustments}
 ]
}
