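# Translucent unlit material for a black hole effect on Niagara or Cascade mesh particles, apparently a temporary asset.
# Emissive blends panning nebula textures through smoke noise, boosts saturation, adds purple tints and Fresnel shading, fades along UV V and multiplies by particle color.
# The alpha chain swirls UV1 with sine/cosine math, spins noise by dynamic-parameter speed, and distorts a circular mask with it.
# That mask is scaled by particle alpha and depth fade, which a platform switch limits to higher-end targets.
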Material /GreaseClipDunmurry/Assets/Temp_M_GreaseClipDunmurry_BlackHole
graph {
 properties {
  bUsedWithMeshParticles=true
  bUsedWithNiagaraMeshParticles=true
  blendMode=BLEND_Translucent
  pixelDepthOffsetMode=PDOM_Legacy
  shadingModel=MSM_Unlit
  usedShadingModels=Unlit
 }
 nodes [
  Add() {constA=1}
  n1 = OneMinus(None=ComponentMask(None=TextureCoordinate()) {g=true})
  n2 = Saturate(None=Multiply(A=Power(Base=@n1) {constExponent=5}) {constB=10})
  n3 = TextureCoordinate() {uTiling=2, vTiling=2.5}
  n4 = Constant2Vector() {g=0.2}
  n5 = TextureSampleParameter2D(UVs=Panner(Coordinate=Add(A=@n3, B=@n4)) {speedY=-0.25}) {
   parameterName=Trail Texture
   texture=/Game/Characters/Player/Female/Medium/Bodies/F_MED_SuperNovaTaro/FX/Material/T_SuperNovaTaro_Skin_Nebula
  }
  n6 = TextureCoordinate() {vTiling=2}
  n7 = Constant2Vector() {g=0.5}
  n8 = TextureSampleParameter2D(UVs=Panner(Coordinate=Add(A=@n6, B=@n7)) {speedX=-0.1, speedY=-0.2}) {
   parameterName=Trail Texture
   texture=&T_SuperNovaTaro_Skin_Nebula
  }
  n9 = TextureCoordinate() {vTiling=2}
  n10 = LinearInterpolate(
   B=Multiply(A=LinearInterpolate(A=@n5.RGB, B=@n8.RGB) {constAlpha=0.7}) {constB=2}
   Alpha=TextureSample(UVs=Panner(Coordinate=@n9) {speedY=-0.1}) {texture=&T_Inky_Smoke_Tile}.R
  )
  n11 = TextureCoordinate() {vTiling=3}
  n12 = AppendVector(A=ComponentMask(None=@n11) {r=true}, B=Power(Base=ComponentMask(None=@n11) {g=true}) {constExponent=0.5})
  n13 = TextureSample(UVs=Panner(Coordinate=TextureCoordinate() {vTiling=2}) {speedY=-0.1}) {samplerType=SAMPLERTYPE_LinearGrayscale, texture=/Game/Effects/Fort_Effects/Emotes/UltraLight/Texture/T_DirectionalNoise_Greyscale}
  n14 = TextureSample(UVs=Add(A=Panner(Coordinate=@n12) {speedX=-0.1, speedY=-0.2}, B=Multiply(A=@n13.R) {constB=0.3})) {texture=&T_SuperNovaTaro_Skin_Nebula}
  n15 = Desaturation(None=Multiply(A=LinearInterpolate(A=@n10, B=@n14.RGB) {constAlpha=0}), Fraction=Constant() {r=-1})
  n16 = Add(A=Constant3Vector() {constant=<rgba 0.1428 0.0254 0.3333 1>}) {constB=0}
  n17 = Constant3Vector() {constant=<rgba 0.0572 0.0432 0.3073 1>}
  n18 = Fresnel() {baseReflectFraction=0, exponent=1}
  n19 = Saturate(None=Multiply(A=Power(Base=@n18) {constExponent=0.5}))
  n20 = Add(
   A=Multiply(A=VectorParameter() {parameterName=color001, defaultValue=<rgba 3.5761 0.9515 4 1>}.RGB, B=Power(Base=@n15) {constExponent=1})
   B=Multiply(A=LinearInterpolate(A=@n16, B=@n17, Alpha=@n19))
  ) {constB=0}
  n21 = Power(Base=Fresnel() {baseReflectFraction=0, exponent=1}) {constExponent=0.1}
  n22 = LinearInterpolate(Alpha=Saturate(None=Multiply(A=@n21))) {constA=5}
  n23 = Multiply(A=Multiply(A=Multiply(A=@n20, B=@n22)) {constB=0.5})
  n24 = ParticleColor()
  MaterialFunctionCall(Emissive=Multiply(A=LinearInterpolate(B=@n23, Alpha=@n2), B=@n24.RGB)) {materialFunction=&MF_FX_EmissiveAdjustments}
  n25 = Add(A=TextureCoordinate() {coordinateIndex=1}, B=Constant2Vector()) {constB=0}
  n26 = SphereMask(A=TextureCoordinate() {coordinateIndex=1}, B=Constant() {r=0.5}, Radius=Constant() {r=1}, Hardness=Constant())
  n27 = Add(A=Multiply(A=Constant() {r=1}) {constB=0.25}, B=Multiply(A=@n26, B=@n26))
  n28 = Cosine(None=@n27)
  n29 = Subtract(A=Multiply(A=TextureCoordinate() {coordinateIndex=1})) {constB=0.5}
  n30 = ComponentMask(None=@n29) {r=true}
  n31 = Multiply(A=@n28, B=@n30)
  n32 = Sine(None=@n27)
  n33 = ComponentMask(None=@n29) {g=true}
  n34 = Multiply(A=@n32, B=@n33)
  n35 = Multiply(A=@n30, B=@n32)
  n36 = Multiply(A=@n33, B=@n28)
  n37 = ScalarParameter() {defaultValue=-1, parameterName=Rotator speed}
  n38 = DynamicParameter() {paramNames [BlackHole size, Rotator Speed, Param3, Param4]}
  n39 = Rotator(
   Coordinate=AppendVector(A=Add(A=@n35, B=@n36), B=Subtract(A=@n31, B=@n34))
   Time=Multiply(A=Multiply(A=@n37, B=@n38.Rotator Speed), B=Time())
  ) {speed=0.3}
  n40 = ComponentMask(None=VectorParameter() {parameterName=UV, defaultValue=<rgba 1 1 1 0>}.RGB) {g=true, r=true}
  n41 = LinearInterpolate(Alpha=TextureSampleParameter2D(UVs=Multiply(A=@n39, B=@n40)) {parameterName=Alpha, texture=&T_NoisePack_01}.RGB) {constA=-0.3}
  n42 = MaterialFunctionCall(Float3=Multiply(A=Power(Base=@n41) {constExponent=1}) {constB=0.2}) {materialFunction=&BreakOutFloat3Components}
  n43 = SphereMask(
   A=Subtract(A=Add(A=@n25, B=@n42.R), B=Multiply(A=@n42.G) {constB=0.1})
   B=Constant() {r=0.5}
   Radius=Constant() {r=0.45}
   Hardness=Constant()
  )
  n44 = Saturate(None=Multiply(A=Power(Base=@n43) {constExponent=1}) {constB=3})
  n45 = DepthFade(FadeDistance=ScalarParameter() {parameterName=DepthFade})
  n46 = Constant() {r=1}
  n47 = MaterialFunctionCall("Xb1,PS4"=@n45, "PC-High&&Epic,XSX,PS5"=@n45, "PC-Perf,Android"=@n46, Switch=@n46, "PC-Low(NotPerf)&&Medium"=@n46) {materialFunction=&PlatformSwitch_1D}
  Multiply(A=Multiply(A=@n24.A, B=Multiply(A=@n44, B=@n47.Result))) {constB=10}
 ]
}
